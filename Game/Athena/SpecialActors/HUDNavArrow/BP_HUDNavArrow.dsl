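; BP_HUDNavArrow is the Athena special-actor Blueprint for a HUD navigation arrow, built around an arrow mesh component.
; SetMaterials takes a dynamic material instance supplied from elsewhere and reattaches or creates it on that arrow component, so the arrow's look is driven by the caller.
; The construction script is empty.
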(fn SetMaterials
  (params (Material%20ID :type Material%20Instance%20Dynamic%20Object%20Reference))
  ;[0.3694 1 0.7653] Apply a material based on the MID being sent in from elsehwere
  (Material|ReattachOrCreateMID (Variables|StaticMesh|GetArrowComponent) Material%20ID))

(fn ConstructionScript)
