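# Niagara effect that draws an amber, semi-transparent glow sprite, apparently the airglow for the HUD navigation arrow.
# Its single Light_And_Glow emitter bursts one particle, with an HLSL spawn probability limiting it to the first loop.
# The particle is not killed by lifetime and persists until the emitter deactivates.
# User parameters for front-end preview and player speed are declared, but no shown module reads them.
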NiagaraSystem /Game/Athena/SpecialActors/HUDNavArrow/FX/NS_HUDNavArrow_Airglow
dependencies {usedDataInterfaces [], usedDynamicInputs [], usedModules [&SystemState]}
emitters {Light_And_Glow {
  emitterUpdateScript [
   &EmitterState()
   &SpawnBurst_Instantaneous(Spawn Count=1, Spawn Probability=HlslExpression() {hlslExpression="Emitter.LoopCount==0"}, Use Spawn Probability=true)
  ]
  particleSpawnScript [/Niagara/Modules/Spawn/Initialization/InitializeParticle(
    Color=<rgba 0.5 0.3667 0 0.6>
    Module.Write Color=true
    Module.Write Lifetime=true
    Module.Write Mass=false
    Module.Write Material Random=true
    Module.Write Position=false
    Module.Write Scale=false
    Module.Write SpriteRotation=false
    Module.Write SpriteSize=true
    Sprite Size=<120 120>
   )]
  particleUpdateScript [&ParticleState(Kill Particles When Lifetime Has Elapsed=false, Let Infinitely Lived Particles Die When Emitter Deactivates=true)]
  rendererClasses [&NiagaraSpriteRendererProperties]
  renderers [{rendererClass=&NiagaraSpriteRendererProperties}]
 }}
summary {
 systemName=NS_HUDNavArrow_Airglow
 emitters [{emitterName=Light_And_Glow, rendererClasses [&NiagaraSpriteRendererProperties]}]
 userVariables [
  User.bIsFrontEndPreview {type=NiagaraBool}
  User.PlayerSpeed {type=NiagaraFloat, defaultValue=0, description=""}
 ]
}
