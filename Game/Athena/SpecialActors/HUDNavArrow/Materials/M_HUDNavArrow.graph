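# Unlit material for the HUD navigation arrow, giving it a two-color holographic look with a separate outline color.
# Inverted fresnel times scanlines, whose block opacity comes from a clamped hologram-wiggle function, plus a panning detail texture blends the two colors.
# A floored U coordinate switches UVs in the upper half of U to the outline color.
# A world-space sphere mask offset from the object position dims the result near that point.
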Material /Game/Athena/SpecialActors/HUDNavArrow/Materials/M_HUDNavArrow
graph {
 properties {pixelDepthOffsetMode=PDOM_Legacy, shadingModel=MSM_Unlit, usedShadingModels=Unlit}
 nodes [
  n1 = VectorParameter() {parameterName=Emissive Color A, defaultValue=<rgba 0.5 0.3167 0 1>}
  n2 = VectorParameter() {parameterName=Emissive Color B, defaultValue=<rgba 2 1.3333 0 1>}
  n3 = Constant() {r=1}
  n4 = ScalarParameter() {defaultValue=3, parameterName=FresnelExponent}
  n5 = ScalarParameter() {defaultValue=76, parameterName=Scanline Amount}
  n6 = ScalarParameter() {defaultValue=-0.15, parameterName=ScanlineSize}
  n7 = ScalarParameter() {defaultValue=0.8, parameterName=Scanlines Dark}
  n8 = ScalarParameter() {defaultValue=1, parameterName=Scanlines Light}
  n9 = Clamp(None=MaterialFunctionCall!(WPO Intensity A, WPO Intensity B, Time, Bias) {materialFunction=/Game/Effects/Fort_Effects/Effects/Atlas/Materials/MF_Atlas_HologramWiggle_01}.Intense 0-1) {maxDefault=0.5}
  n10 = TextureCoordinate()
  n11 = AppendVector(A=ScalarParameter() {defaultValue=1, parameterName=DetailTile-X}, B=ScalarParameter() {defaultValue=0.3, parameterName=DetailTile-Y})
  n12 = ScalarParameter() {parameterName=DetailPan-X}
  n13 = ScalarParameter() {defaultValue=1, parameterName=DetailPan-Y}
  n14 = TextureSampleParameter2D(UVs=Panner(Coordinate=Multiply(A=@n10, B=@n11), Speed=AppendVector(A=@n12, B=@n13))) {
   parameterName=Detail-Texture
   samplerSource=SSM_Wrap_WorldGroupSettings
   texture=/Game/Gadgets/Assets/AdrenalineRush/Textures/T_AdrenalineRush_HealingGradient_01
  }
  n15 = Add(
   A=Multiply(
    A=MaterialFunctionCall(Brightness=@n3, Exponent=@n4) {materialFunction=&MF_Fresnel_FX}.Fresnel Inverted
    B=MaterialFunctionCall!(
     Number of Scanlines=@n5
     Scanline Speed
     "Scanlines Thickness (Bias)"=@n6
     Scanlines Dark Value=@n7
     Scanlines Light Value=@n8
     Block Opacity=@n9
     UVs
     Time
     Block Repeat
    ) {materialFunction=/Game/Effects/Fort_Effects/Effects/Atlas/Materials/MF_Atlas_Scanlines_01}.Result
   )
   B=Multiply(A=@n14.RGB, B=VectorParameter() {parameterName=Detail-Intensity, defaultValue=<rgba 1 1 1 1>}.RGB)
  )
  n16 = Multiply(A=ComponentMask(None=TextureCoordinate()) {r=true}) {constB=2}
  n17 = SphereMask(
   A=WorldPosition().XYZ
   B=Add(A=ObjectPositionWS(), B=VectorParameter() {parameterName=SphereMask_Color-Position, defaultValue=<rgba -14 0 0 1>}.RGB)
   Radius=ScalarParameter() {defaultValue=115, parameterName=SphereMask_Color-Radius}
   Hardness=ScalarParameter() {parameterName=SphereMask_Color-Hardness}
  )
  n18 = ScalarParameter() {defaultValue=1, parameterName=SphereMask}
  Multiply(
   A=LinearInterpolate(
    A=LinearInterpolate(A=@n1.RGB, B=@n2.RGB, Alpha=@n15)
    B=VectorParameter() {parameterName=Color-Outline, defaultValue=<rgba 1.468 1.7517 2 1>}.RGB
    Alpha=Floor(None=@n16)
   )
   B=LinearInterpolate(
    A=ScalarParameter() {defaultValue=1, parameterName=SphereMsk_Color-LowValue}
    B=ScalarParameter() {defaultValue=0.3, parameterName=SphereMask_Color-HighValue}
    Alpha=Power(Base=@n17, Exp=@n18)
   )
  )
  ScalarParameter() {defaultValue=1}
 ]
}
